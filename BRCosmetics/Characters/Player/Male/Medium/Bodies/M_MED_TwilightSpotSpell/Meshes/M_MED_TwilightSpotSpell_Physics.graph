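# Physics asset for the TwilightSpotSpell medium male player body cosmetic, covering the suit's dangling flaps and the lower body around them.
# Capsules on pelvis, spine_01 and both thighs accompany six two-bone flap chains, also built from capsules.
# Back and mid-back chains hang from the pelvis and side chains from the thighs, with tight angular limits.
# A free-motion constraint also links the right mid-back and back flap tips, with no left-side counterpart.
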PhysicsAsset /BRCosmetics/Characters/Player/Male/Medium/Bodies/M_MED_TwilightSpotSpell/Meshes/M_MED_TwilightSpotSpell_Physics
bodies {
 dyn_suitFlap_bck_l_01 [{
   shapeType=Capsule
   length=4.5
   radius=1.3
   shapeName=dyn_suitFlap_bck_l_01_capsule
   center=<2.8816 0 1.0038>
   rotation=<rot -90 179.8813 -179.8813>
  }]
 dyn_suitFlap_bck_l_02 [{shapeType=Capsule, length=3.5, radius=1.9, shapeName=dyn_suitFlap_bck_l_02_capsule, center=<2.8816 0 1.0038>, rotation=<rot -90 90 -89.9999>}]
 dyn_suitFlap_bck_r_01 [{
   shapeType=Capsule
   length=4.5
   radius=1.3
   shapeName=dyn_suitFlap_bck_r_01_capsule
   center=<2.8816 0 1.0038>
   rotation=<rot -90 -179.9975 179.9975>
  }]
 dyn_suitFlap_bck_r_02 [{
   shapeType=Capsule
   length=3.5
   radius=1.9
   shapeName=dyn_suitFlap_bck_r_02_capsule
   center=<2.8816 0 1.0038>
   rotation=<rot -90 -178.7341 178.7341>
  }]
 dyn_suitFlap_midBck_l_01 [{
   shapeType=Capsule
   length=3
   radius=1.8
   shapeName=dyn_suitFlap_midBck_l_01_capsule
   center=<2.5243 0 1.2236>
   rotation=<rot 90 -0.0023 -0.0025>
  }]
 dyn_suitFlap_midBck_l_02 [{
   shapeType=Capsule
   length=3
   radius=2.4
   shapeName=dyn_suitFlap_midBck_l_02_capsule
   center=<2.5243 0 1.2236>
   rotation=<rot 90 0.0001 -0.0001>
  }]
 dyn_suitFlap_midBck_r_01 [{shapeType=Capsule, length=3, radius=1.8, shapeName=dyn_suitFlap_midBck_r_01_capsule, center=<2.5243 0 1.2236>, rotation=<rot 90 90 90.0002>}]
 dyn_suitFlap_midBck_r_02 [{
   shapeType=Capsule
   length=3
   radius=2.4
   shapeName=dyn_suitFlap_midBck_r_02_capsule
   center=<2.5243 0 1.2236>
   rotation=<rot 90 0.0001 0.0003>
  }]
 dyn_suitFlap_side_l_01 [{shapeType=Capsule, length=3, radius=1.8, shapeName=dyn_suitFlap_side_l_01_capsule, center=<2.5243 0 1.2236>, rotation=<rot 90 -179.9998 180>}]
 dyn_suitFlap_side_l_02 [{shapeType=Capsule, length=3, radius=2.4, shapeName=dyn_suitFlap_side_l_02_capsule, center=<2.5243 0 1.2236>, rotation=<rot 90 0 -0.0003>}]
 dyn_suitFlap_side_r_01 [{shapeType=Capsule, length=3, radius=1.8, shapeName=dyn_suitFlap_side_r_01_capsule, center=<2.5243 0 1.2236>, rotation=<rot 90 180 -179.9997>}]
 dyn_suitFlap_side_r_02 [{shapeType=Capsule, length=3, radius=2.4, shapeName=dyn_suitFlap_side_r_02_capsule, center=<2.5243 0 1.2236>, rotation=<rot 90 180 -179.9997>}]
 pelvis [{shapeType=Capsule, length=9.5589, radius=10.1478, shapeName=pelvis_capsule}]
 spine_01 [{shapeType=Capsule, length=8.6647, radius=9.6839, shapeName=spine_01_capsule, center=<-1.8965 0.2311 0.0146>}]
 thigh_l [{
   shapeType=Capsule
   length=16.9757
   radius=8.1356
   shapeName=thigh_l_capsule
   center=<-10.0243 -1.2805 0.7692>
   rotation=<rot 0 -89.3927 -84.3029>
  }]
 thigh_r [{
   shapeType=Capsule
   length=17.0731
   radius=8.0103
   shapeName=thigh_r_capsule
   center=<10.0054 1.7479 -0.9548>
   rotation=<rot 0 -90.1214 -84.5334>
  }]
}
constraints [
 {
  bone1Name=spine_01
  bone2Name=pelvis
  swing1LimitDegrees=45
  swing1Motion=Free
  swing2LimitDegrees=45
  swing2Motion=Free
  twistLimitDegrees=45
  twistMotion=Free
 }
 {
  bone1Name=thigh_l
  bone2Name=pelvis
  swing1LimitDegrees=45
  swing1Motion=Free
  swing2LimitDegrees=45
  swing2Motion=Free
  twistLimitDegrees=45
  twistMotion=Free
 }
 {
  bone1Name=thigh_r
  bone2Name=pelvis
  swing1LimitDegrees=45
  swing1Motion=Free
  swing2LimitDegrees=45
  swing2Motion=Free
  twistLimitDegrees=45
  twistMotion=Free
 }
 {bone1Name=dyn_suitFlap_bck_l_01, bone2Name=pelvis, swing1LimitDegrees=10, swing2LimitDegrees=30, twistLimitDegrees=10}
 {bone1Name=dyn_suitFlap_bck_l_02, bone2Name=dyn_suitFlap_bck_l_01, swing1LimitDegrees=10, swing2LimitDegrees=30, twistLimitDegrees=10}
 {bone1Name=dyn_suitFlap_bck_r_01, bone2Name=pelvis, swing1LimitDegrees=10, swing2LimitDegrees=30, twistLimitDegrees=10}
 {bone1Name=dyn_suitFlap_bck_r_02, bone2Name=dyn_suitFlap_bck_r_01, swing1LimitDegrees=10, swing2LimitDegrees=30, twistLimitDegrees=10}
 {bone1Name=dyn_suitFlap_side_r_01, bone2Name=thigh_r, swing1LimitDegrees=10, swing2LimitDegrees=40, twistLimitDegrees=0}
 {bone1Name=dyn_suitFlap_side_r_02, bone2Name=dyn_suitFlap_side_r_01, swing1LimitDegrees=10, swing2LimitDegrees=60, twistLimitDegrees=0}
 {bone1Name=dyn_suitFlap_side_l_02, bone2Name=dyn_suitFlap_side_l_01, swing1LimitDegrees=10, swing2LimitDegrees=60, twistLimitDegrees=0}
 {bone1Name=dyn_suitFlap_side_l_01, bone2Name=thigh_l, swing1LimitDegrees=10, swing2LimitDegrees=40, twistLimitDegrees=0}
 {bone1Name=dyn_suitFlap_midBck_r_02, bone2Name=dyn_suitFlap_midBck_r_01, swing1LimitDegrees=10, swing2LimitDegrees=60, twistLimitDegrees=0}
 {bone1Name=dyn_suitFlap_midBck_r_01, bone2Name=pelvis, swing1LimitDegrees=20, swing2LimitDegrees=40, twistLimitDegrees=0}
 {bone1Name=dyn_suitFlap_midBck_l_02, bone2Name=dyn_suitFlap_midBck_l_01, swing1LimitDegrees=10, swing2LimitDegrees=60, twistLimitDegrees=0}
 {bone1Name=dyn_suitFlap_midBck_l_01, bone2Name=pelvis, swing1LimitDegrees=20, swing2LimitDegrees=40, twistLimitDegrees=0}
 {
  bone1Name=dyn_suitFlap_midBck_r_02
  bone2Name=dyn_suitFlap_bck_r_02
  swing1LimitDegrees=45
  swing1Motion=Free
  swing2LimitDegrees=45
  swing2Motion=Free
  twistLimitDegrees=45
  twistMotion=Free
 }
]
