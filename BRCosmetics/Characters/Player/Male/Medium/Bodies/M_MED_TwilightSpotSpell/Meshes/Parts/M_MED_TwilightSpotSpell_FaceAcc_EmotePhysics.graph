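# Physics asset for the TwilightSpotSpell medium male face accessory, named for emote use, that sets up two dangling scarf chains.
# Small capsules on five left and six right dyn_scarf bones are constrained in series, each chain rooted at spine_05.
# Root joints have tight single-axis swing limits, later links have wider swing limits, and all have zero twist limits.
# Larger capsules on the spine, clavicles and arms give body shapes for the scarf.
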PhysicsAsset /BRCosmetics/Characters/Player/Male/Medium/Bodies/M_MED_TwilightSpotSpell/Meshes/Parts/M_MED_TwilightSpotSpell_FaceAcc_EmotePhysics
bodies {
 clavicle_l [{
   shapeType=Capsule
   length=3.7252
   radius=7.6208
   shapeName=clavicle_l_capsule
   center=<7.8802 0 0.6481>
   rotation=<rot 90 -179.9998 179.9996>
  }]
 clavicle_r [{
   shapeType=Capsule
   length=5.5
   radius=6.995
   shapeName=clavicle_r_capsule
   center=<-9.0595 0.0001 -0.9047>
   rotation=<rot -90 179.9978 -179.9986>
  }]
 dyn_scarf_l_01 [{shapeType=Capsule, length=2.5, radius=2, shapeName=dyn_scarf_l_01_capsule, center=<2.6223 0 1.1869>, rotation=<rot -90 0.0001 -0.0001>}]
 dyn_scarf_l_02 [{shapeType=Capsule, length=2.5, radius=2, shapeName=dyn_scarf_l_02_capsule, center=<2.6223 0 0.6422>, rotation=<rot -90 -179.9736 179.9737>}]
 dyn_scarf_l_03 [{shapeType=Capsule, length=2.5, radius=2.5, shapeName=dyn_scarf_l_03_capsule, center=<2.6223 0 0.6422>, rotation=<rot -90 173.2786 -173.2786>}]
 dyn_scarf_l_04 [{shapeType=Capsule, length=2.5, radius=2.5, shapeName=dyn_scarf_l_04_capsule, center=<2.6223 0 0.6422>, rotation=<rot -90 180 -179.9999>}]
 dyn_scarf_l_05 [{shapeType=Capsule, length=2.5, radius=2.5, shapeName=dyn_scarf_l_05_capsule, center=<2.6223 0 0.6422>, rotation=<rot -90 -180 180>}]
 dyn_scarf_r_01 [{shapeType=Capsule, length=2.5, radius=2, shapeName=dyn_scarf_r_01_capsule, center=<2.6223 0 1.1869>, rotation=<rot -90 -179.9604 179.9604>}]
 dyn_scarf_r_02 [{shapeType=Capsule, length=2.5, radius=2, shapeName=dyn_scarf_r_02_capsule, center=<2.6223 0 0.3615>, rotation=<rot -90 0 0>}]
 dyn_scarf_r_03 [{shapeType=Capsule, length=2.5, radius=2.5, shapeName=dyn_scarf_r_03_capsule, center=<2.6223 0 0.3615>, rotation=<rot -90 -179.9994 179.9994>}]
 dyn_scarf_r_04 [{shapeType=Capsule, length=2.5, radius=2.5, shapeName=dyn_scarf_r_04_capsule, center=<2.6223 0 0.3615>, rotation=<rot -90 180 180>}]
 dyn_scarf_r_05 [{shapeType=Capsule, length=2.5, radius=2.5, shapeName=dyn_scarf_r_05_capsule, center=<2.6223 0 0.3615>, rotation=<rot -90 90 -90>}]
 dyn_scarf_r_06 [{shapeType=Capsule, length=2.5, radius=2.5, shapeName=dyn_scarf_r_06_capsule, center=<2.6223 0 0.3615>, rotation=<rot -90 -90 90>}]
 lowerarm_l [{
   shapeType=Capsule
   length=19.5
   radius=5.505
   shapeName=lowerarm_l_capsule
   center=<9.1352 0.8679 0>
   rotation=<rot -0.0004 -89.9996 89.9998>
  }]
 lowerarm_r [{
   shapeType=Capsule
   length=19.5
   radius=5.505
   shapeName=lowerarm_r_capsule
   center=<-9.1353 -0.5875 0>
   rotation=<rot -0.0001 -89.9999 -90.0001>
  }]
 spine_02 [{shapeType=Capsule, length=7.005, radius=10.505, shapeName=spine_02_capsule, center=<-5.2066 0 0>}]
 spine_03 [{shapeType=Capsule, length=10, radius=10.505, shapeName=spine_03_capsule}]
 spine_04 [{shapeType=Capsule, length=15.5, radius=9.505, shapeName=spine_04_capsule, center=<0.7785 0.7551 0>}]
 spine_05 [{shapeType=Capsule, length=17.708, radius=9.7356, shapeName=spine_05_capsule}]
 upperarm_l [{shapeType=Capsule, length=16, radius=6.5, shapeName=upperarm_l_capsule, center=<10.4589 0 -1.6389>, rotation=<rot -0.0002 -89.9998 90>}]
 upperarm_r [{shapeType=Capsule, length=28, radius=6, shapeName=upperarm_r_capsule, center=<-13.6812 0.643 1.6389>, rotation=<rot -0.0002 -89.9998 -90>}]
}
constraints [
 {bone1Name=dyn_scarf_l_01, bone2Name=spine_05, swing1LimitDegrees=0, swing2LimitDegrees=20, twistLimitDegrees=0}
 {bone1Name=dyn_scarf_l_02, bone2Name=dyn_scarf_l_01, swing1LimitDegrees=60, swing2LimitDegrees=60, twistLimitDegrees=0}
 {bone1Name=dyn_scarf_l_03, bone2Name=dyn_scarf_l_02, swing1LimitDegrees=60, swing2LimitDegrees=60, twistLimitDegrees=0}
 {bone1Name=dyn_scarf_l_04, bone2Name=dyn_scarf_l_03, swing1LimitDegrees=60, swing2LimitDegrees=60, twistLimitDegrees=0}
 {bone1Name=dyn_scarf_l_05, bone2Name=dyn_scarf_l_04, swing1LimitDegrees=60, swing2LimitDegrees=60, twistLimitDegrees=0}
 {bone1Name=dyn_scarf_r_01, bone2Name=spine_05, swing1LimitDegrees=0, swing2LimitDegrees=20, twistLimitDegrees=0}
 {bone1Name=dyn_scarf_r_02, bone2Name=dyn_scarf_r_01, swing1LimitDegrees=60, swing2LimitDegrees=60, twistLimitDegrees=0}
 {bone1Name=dyn_scarf_r_03, bone2Name=dyn_scarf_r_02, swing1LimitDegrees=60, swing2LimitDegrees=60, twistLimitDegrees=0}
 {bone1Name=dyn_scarf_r_04, bone2Name=dyn_scarf_r_03, swing1LimitDegrees=60, swing2LimitDegrees=60, twistLimitDegrees=0}
 {bone1Name=dyn_scarf_r_05, bone2Name=dyn_scarf_r_04, swing1LimitDegrees=60, swing2LimitDegrees=60, twistLimitDegrees=0}
 {bone1Name=dyn_scarf_r_06, bone2Name=dyn_scarf_r_05, swing1LimitDegrees=60, swing2LimitDegrees=60, twistLimitDegrees=0}
]
